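# Physics asset for the basic SUV vehicle body cosmetic, named as a BR variant, that gives the body mesh its simplified collision representation.
# It holds a single body on the frame bone, with no collision shape values shown in the file.
# It defines no constraints, so nothing in the asset is jointed or articulated.
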PhysicsAsset /VehicleCosmetics/Bodies/_SUV/Basic/Meshes/Body_SUV_Basic_PhysicsBR
bodies {frame []}
constraints []
