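# Physics asset for the basic SUV body mesh, named as a DM variant.
# It holds a single root body with one box shape sized to the vehicle and raised above the origin.
# It has no constraints, so the body collides as one rigid box.
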PhysicsAsset /VehicleCosmetics/Bodies/_SUV/Basic/Meshes/Body_SUV_Basic_PhysicsDM
bodies {root [{shapeType=Box, extentX=444, extentY=216, extentZ=116, shapeName=root_box, center=<0 0 93.2685>}]}
constraints []
